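# Niagara system for the swing trail of the PaintSplash bass pickaxe, a pink paint effect.
# Two AnimTrail ribbons follow mesh bones: a randomized galaxy trail and a paint trail whose width and material are shaped over emitter age.
# Sprite emitters spawn by distance moved, adding SubUV paint splashes and speed-stretched droplets under gravity and drag.
# Paint trail, splash and droplet colors switch with time of day.
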NiagaraSystem /SparksCosmetics/Instrument/Bass/PaintSplash/Effects/Pickaxe/NS_Pickaxe_Bass_PaintSplash_Era_Trail
dependencies {
 usedDataInterfaces [
  &NiagaraDataInterfaceCurve
  &NiagaraDataInterfaceSpriteRendererInfo
  &NiagaraDataInterfaceColorCurve
 ]
 usedDynamicInputs []
 usedModules [&SystemState]
}
emitters {
 Droplets {
  emitterUpdateScript [
   &EmitterState()
   &SpawnPerUnit(Module.Use Max Movement Threshold=true, Module.Use Movement Tolerance=false, Spawn Spacing=15, Use Spawn Probability=false)
  ]
  particleSpawnScript [
   &InitializeParticle(
    Color=&FN_Select_Color_from_Time_Of_Day(
     Morning=<rgba 0.6354 0.0691 0.442 1>
     Day=<rgba 0.7292 0.0793 0.5072 1>
     Evening=<rgba 0.3608 0.0392 0.251 1>
     Night=<rgba 0.6146 0 0.1074 1>
    )
    Lifetime=&RandomRangeFloat(Minimum=0.3, Maximum=0.5)
    Sprite Size Mode=::NewEnumerator3
    Uniform Sprite Size=&RandomRangeFloat(Minimum=1, Maximum=3.5)
    UsePositionOffset=false
    Write Material Random=true
   )
   &AddVelocity(
    Module.Use Falloff Distance=false
    Override Seed=false
    Velocity Mode=::NewEnumerator1
    Velocity Speed=&RandomRangeFloat(Minimum=200, Maximum=400)
   )
   &InheritVelocity(Speed Limit=true, UseSpeedThreshold=false) {enabled=false}
   &ShapeLocation(MaskSpawnGroup=false, Offset=<0 0 -10>, Offset Mode=::NewEnumerator0, Override Seed=false, Sphere Radius=2, UseOwnerScale=true)
  ]
  particleUpdateScript [
   &GravityForce()
   &Drag(Drag=&RandomRangeFloat(Minimum=4, Maximum=8), Use Linear Drag=true, Use Rotational Drag=true)
   &SolveForcesAndVelocity(Clamp Velocity=false, Limit Acceleration=false)
   &ParticleState()
   &ScaleSpriteSize(Uniform Curve Sprite Scale=DataInterface() {propertyValues {Curve {keys [
        {arriveTangent=4.6571, interpMode=RCIM_Cubic, leaveTangent=4.6571, tangentMode=RCTM_User, time=0, value=0}
        {interpMode=RCIM_Cubic, tangentMode=RCTM_User, time=0.5, value=1}
        {arriveTangent=-5.1829, tangentMode=RCTM_User, time=1, value=0}
       ]}}})
   &ScaleSpriteSizeBySpeed(Max Scale Factor=<1 5>, Min Scale Factor=<1 1.5>)
  ]
  rendererClasses [&NiagaraSpriteRendererProperties]
  renderers [{rendererClass=&NiagaraSpriteRendererProperties}]
 }
 Galaxy_Trail {
  emitterSpawnScript [/SetVariables(Emitter.Random ID:NiagaraInt32=&RandomRangeInt(Minimum=0, Maximum=100, Randomness Mode=::NewEnumerator2, Random Seed=?)) {isSetParameters=true}]
  emitterUpdateScript [
   &EmitterState()
   &FN_SpawnAnimTrail_v2()
   /SetVariables(Emitter.RandomFloat:NiagaraFloat=&RandomRangeFloat(Minimum=0, Maximum=1)) {isSetParameters=true}
  ]
  particleSpawnScript [
   /Game/Effects/Niagara/Modules/AnimTrail/FN_InitializeAnimTrail/FN_InitializeAnimTrailBoneIndecies(Second Bone or Socket Index=1)
   /Game/Effects/Niagara/Modules/AnimTrail/FN_InitializeAnimTrail/FN_InitializeAnimTrail(
    AnimTrailWidth=1.4
    Color=<rgba 1 0.516 0.8903 1>
    Lifetime=0.4
    Position Offset:Vector3f=&RotateVectorByQuaternion(Vector To Rotate=<0 0 10>, Quaternion=$Engine.Owner.Rotation)
   )
   &DynamicMaterialParameters(
    Index 0 Param 1=$Emitter.RandomFloat
    Index 0 Param 2=$System.RandomFloatB
    Module.Param0WriteEnabled=true
    Module.Param1WriteEnabled=false
    Module.Param2WriteEnabled=false
    Module.Param3WriteEnabled=false
   )
   /SetVariables(Particles.SpawnGroup=&MakeIntFromFloat(Float=&Multiply_Float(A=$System.RandomFloatA, B=100))) {isSetParameters=true}
   /SetVariables(Particles.RibbonID=/Niagara/DynamicInputs/Meshes/MakeRibbonIDFromSpawnGroup()) {isSetParameters=true}
  ]
  particleUpdateScript [
   &FN_UpdateAnimTrail_v2()
   &ScaleColor(
    Scale Alpha=&FloatFromCurve(FloatCurve=DataInterface() {propertyValues {Curve {keys [
         {arriveTangent=0.0035, interpMode=RCIM_Cubic, leaveTangent=0.0035, tangentMode=RCTM_User, time=0, value=1.1}
         {arriveTangent=0.0081, interpMode=RCIM_Cubic, leaveTangent=0.0081, tangentMode=RCTM_User, time=1, value=0}
        ]}}})
    ScaleA=true
    ScaleRGB=true
   )
   &ScaleRibbonWidth(Ribbon Width Scale=&FloatFromCurve(FloatCurve=DataInterface() {propertyValues {Curve {keys [
         {interpMode=RCIM_Cubic, tangentMode=RCTM_User, time=0, value=1}
         {arriveTangent=-2.9482, interpMode=RCIM_Cubic, leaveTangent=-2.9482, tangentMode=RCTM_User, time=1, value=0}
        ]}}}))
   &DynamicMaterialParameters(
    Index 0 Param 1=$Emitter.RandomFloat
    Index 0 Param 2=&FloatFromCurve(FloatCurve=DataInterface() {propertyValues {Curve {keys [{interpMode=RCIM_Cubic, tangentMode=RCTM_User, time=0.1, value=0}]}}})
    Index 0 Param 4=&FloatFromCurve(FloatCurve=DataInterface() {propertyValues {Curve {keys [
         {interpMode=RCIM_Cubic, tangentMode=RCTM_User, time=0.1, value=1}
         {arriveTangent=-0.0076, interpMode=RCIM_Cubic, leaveTangent=-0.0076, tangentMode=RCTM_User, time=0.8, value=0}
        ]}}})
    Module.Param0WriteEnabled=false
    Module.Param1WriteEnabled=false
    Module.Param2WriteEnabled=false
    Module.Param3WriteEnabled=true
   ) {module=DynamicMaterialParameters001}
  ]
  rendererClasses [&NiagaraRibbonRendererProperties]
  renderers [{rendererClass=&NiagaraRibbonRendererProperties}]
 }
 PaintSplash001 {
  emitterUpdateScript [
   &EmitterState()
   &SpawnPerUnit(Module.Use Max Movement Threshold=true, Module.Use Movement Tolerance=false, Spawn Spacing=30, Use Spawn Probability=false)
  ]
  particleSpawnScript [
   &InitializeParticle(
    Color=&FN_Select_Color_from_Time_Of_Day(
     Morning=<rgba 0.2135 0.0232 0.1486 1>
     Day=<rgba 0.3608 0.0392 0.251 1>
     Evening=<rgba 0.3608 0.0392 0.251 1>
     Night=<rgba 0.1406 0.0153 0.0978 1>
    )
    Lifetime=&RandomRangeFloat(Minimum=0.2, Maximum=0.25)
    Position Offset=&RotateVectorByQuaternion(Vector To Rotate=<0 0 -10>, Quaternion=$Engine.Owner.Rotation)
    Sprite Rotation Mode=::NewEnumerator2
    Sprite Size Mode=::NewEnumerator3
    Uniform Sprite Size=&RandomRangeFloat(Minimum=50, Maximum=65)
    UsePositionOffset=true
    Write Material Random=true
   )
   &SubUVAnimation(
    Evaluation Type=::NewEnumerator0
    "Sprite Renderer "=DataInterface() {propertyValues {SpriteRenderer=&self:PaintSplash001_0.NiagaraSpriteRendererProperties_0}}
    SubUV Animation Mode:ENiagara_SubUVLookupModeV2=::NewEnumerator2
    UseEndFrame=false
    UseStartFrame=false
   )
   &ShapeLocation(MaskSpawnGroup=false, Override Seed=false, Sphere Radius=3, UseOwnerScale=true)
   &AddVelocity(
    Module.Use Falloff Distance=false
    Override Seed=false
    Velocity Mode=::NewEnumerator1
    Velocity Speed=&RandomRangeFloat(Minimum=50, Maximum=100)
   )
  ]
  particleUpdateScript [
   &ParticleState()
   &ScaleSpriteSize(Uniform Curve Sprite Scale=DataInterface() {propertyValues {Curve {keys [
        {arriveTangent=2.7329, interpMode=RCIM_Cubic, leaveTangent=2.7329, tangentMode=RCTM_User, time=0, value=0}
        {arriveTangent=0.0224, interpMode=RCIM_Cubic, leaveTangent=0.0224, tangentMode=RCTM_User, time=1, value=1}
       ]}}})
   &GravityForce(Gravity=&RandomRangeVector(Minimum=<0 0 -100>, Maximum=<0 0 -200>))
   &Drag(Drag=&RandomRangeFloat(Minimum=0, Maximum=1), Use Linear Drag=true, Use Rotational Drag=true)
   &SolveForcesAndVelocity(Clamp Velocity=false, Limit Acceleration=false)
   &DynamicMaterialParameters(
    Index 0 Param 4=&FloatFromCurve(FloatCurve=DataInterface() {propertyValues {Curve {keys [<at 0 0>, {arriveTangent=1.8421, leaveTangent=1.8421, time=1, value=1}]}}})
    Module.Index0Float0Write=true
    Module.Index0Float1Write=true
    Module.Index0Float2Write=true
    Module.Index0Float3Write=true
   )
   &ScaleColor(
    Scale RGB=&Multiply_Vector(
     A=&MakeVectorFromLinearColorRGB(Color=&ColorFromCurve(ColorCurve=DataInterface() {propertyValues {
         AlphaCurve {keys [<at 0 1 RCIM_Cubic>]}
         BlueCurve {keys [<at 1 0.1804 RCIM_Cubic>]}
         GreenCurve {keys [<at 1 0.0745 RCIM_Cubic>]}
         RedCurve {keys [<at 1 0 RCIM_Cubic>]}
        }}))
     B=&VectorFromFloat(Value=1)
    )
    ScaleA=true
    ScaleRGB=true
   ) {enabled=false}
  ]
  rendererClasses [&NiagaraSpriteRendererProperties]
  renderers [{rendererClass=&NiagaraSpriteRendererProperties}]
 }
 PaintTrail001 {
  emitterUpdateScript [
   &EmitterState()
   &FN_SpawnAnimTrail_v2()
  ]
  particleSpawnScript [
   &FN_InitializeAnimTrailBoneIndecies(First Bone or Socket Index=1)
   &FN_InitializeAnimTrail(
    AnimTrailWidth=1
    Color=&ColorFromCurve(
     ColorCurve=DataInterface() {propertyValues {
       AlphaCurve {keys [<at 0 0>,<at 0.036 0.7>,<at 0.062 1 RCIM_Cubic>,<at 0.094 0.7>,<at 0.138 0 RCIM_Cubic>]}
       BlueCurve {keys [<at 0 1 RCIM_Cubic>,<at 1 1 RCIM_Cubic>]}
       GreenCurve {keys [<at 0 1 RCIM_Cubic>,<at 1 1 RCIM_Cubic>]}
       RedCurve {keys [<at 0 1 RCIM_Cubic>,<at 1 1 RCIM_Cubic>]}
      }}
     CurveIndex=$Emitter.Age
    )
    Lifetime=0.4
    Position Offset:Vector3f=&RotateVectorByQuaternion(Vector To Rotate=<0 0 -10>, Quaternion=$Engine.Owner.Rotation)
   )
   &SystemLocation(MaskSpawnGroup=false)
   &DynamicMaterialParameters(
    Index 0 Param 2=&FloatFromCurve(
     FloatCurve=DataInterface() {propertyValues {Curve {keys [
         {arriveTangent=59.9219, interpMode=RCIM_Cubic, leaveTangent=59.9219, tangentMode=RCTM_User, time=0, value=0}
         {arriveTangent=15.2822, leaveTangent=15.2822, time=0.0248, value=0.9936}
         {arriveTangent=-8.7802, interpMode=RCIM_Cubic, leaveTangent=-8.7802, time=0.1047, value=0.9965}
         {arriveTangent=-24.5447, interpMode=RCIM_Cubic, leaveTangent=-24.5447, tangentMode=RCTM_User, time=0.138, value=0}
        ]}}}
     CurveIndex=$Emitter.Age
    )
    Module.Index0Float0Write=true
    Module.Index0Float1Write=true
    Module.Index0Float2Write=true
    Module.Index0Float3Write=true
   )
   /SetVariables(Particles.Color=&ColorFromCurve(
     ColorCurve=DataInterface() {propertyValues {
       AlphaCurve {keys [<at 0.1629 1 RCIM_Cubic>]}
       BlueCurve {keys [<at 0.0237 0 RCIM_Cubic>,<at 0.0757 0>,<at 0.1385 1 RCIM_Cubic>]}
       GreenCurve {keys [<at 0.0237 0 RCIM_Cubic>,<at 0.0757 1>,<at 0.1385 0 RCIM_Cubic>]}
       RedCurve {keys [<at 0.0237 1 RCIM_Cubic>,<at 0.0757 0>,<at 0.1385 0.0042 RCIM_Cubic>]}
      }}
     CurveIndex=$Emitter.Age
    )) {enabled=false, isSetParameters=true}
   /SetVariables(Particles.RibbonWidth=&FloatFromCurve(
     FloatCurve=DataInterface() {propertyValues {Curve {keys [
         {arriveTangent=27.5119, interpMode=RCIM_Cubic, leaveTangent=27.5119, tangentMode=RCTM_User, time=0, value=0}
         {interpMode=RCIM_Cubic, tangentMode=RCTM_User, time=0.0644, value=1}
         {arriveTangent=-24.0579, tangentMode=RCTM_User, time=0.138, value=0}
        ]}}}
     CurveIndex=$Emitter.Age
     Scale Curve=65
    )) {isSetParameters=true}
  ]
  particleUpdateScript [
   &ParticleState()
   &FN_UpdateAnimTrail_v2()
   &DynamicMaterialParameters(
    Index 0 Param 1=&RandomRangeFloat(Minimum=0, Maximum=0.5)
    Index 0 Param 2=&FloatFromCurve(
     FloatCurve=DataInterface() {propertyValues {Curve {keys [
         {arriveTangent=-0.0877, interpMode=RCIM_Cubic, leaveTangent=-0.0877, tangentMode=RCTM_User, time=-0.2, value=1}
         {arriveTangent=-2.7712, interpMode=RCIM_Cubic, leaveTangent=-2.7712, tangentMode=RCTM_User, time=0.35, value=0}
        ]}}}
     CurveIndex=$Emitter.Age
     Scale Curve=&MakeFloatFromVector4(Vector 4=$Particles.Initial.DynamicMaterialParameter, Channel=::NewEnumerator1)
    )
    Index 0 Param 3=1
    Index 0 Param 4=1
    Module.Index0Float0Write=false
    Module.Index0Float1Write=true
    Module.Index0Float2Write=true
    Module.Index0Float3Write=true
   ) {module=DynamicMaterialParameters001}
   &ScaleColor(
    Scale Alpha=&Multiply_Float(
     A=&FloatFromCurve(
      FloatCurve=DataInterface() {propertyValues {Curve {keys [
          {interpMode=RCIM_Cubic, tangentMode=RCTM_User, time=0, value=1}
          {interpMode=RCIM_Cubic, tangentMode=RCTM_User, time=0.2, value=1}
          {arriveTangent=-13.2192, tangentMode=RCTM_User, time=0.35, value=0}
         ]}}}
      CurveIndex=$Emitter.Age
     )
     B=&FNTimeOfDayScalar(Night Scale=1, Morning Scale=1, Evening Scale=1)
    )
    ScaleA=false
    ScaleRGB=true
   )
   &ScaleColor(
    Scale Alpha=&FloatFromCurve(FloatCurve=DataInterface() {propertyValues {Curve {keys [
         {arriveTangent=5.0128, interpMode=RCIM_Cubic, leaveTangent=5.0128, tangentMode=RCTM_User, time=0, value=0}
         {interpMode=RCIM_Cubic, tangentMode=RCTM_User, time=0.5, value=1}
         {arriveTangent=-4.8568, tangentMode=RCTM_User, time=1, value=0}
        ]}}})
    Scale RGB=&MakeVectorFromLinearColorRGB(Color=&FN_Select_Color_from_Time_Of_Day(
      Morning=<rgba 0.7552 0.0821 0.5254 1>
      Day=<rgba 0.5 0.0543 0.3478 1>
      Evening=<rgba 0.3608 0.0392 0.251 1>
      Night=<rgba 0.224 0.0243 0.1558 1>
     ))
    ScaleA=false
    ScaleRGB=true
   ) {module=ScaleColor001}
   &ScaleRibbonWidth(Ribbon Width Scale=&FloatFromCurve(FloatCurve=DataInterface() {propertyValues {Curve {keys [
         {arriveTangent=3.5714, interpMode=RCIM_Cubic, leaveTangent=3.5714, tangentMode=RCTM_User, time=0, value=0}
         {arriveTangent=-0.0501, interpMode=RCIM_Cubic, leaveTangent=-0.0501, tangentMode=RCTM_User, time=0.5, value=1}
         {arriveTangent=-3.4513, tangentMode=RCTM_User, time=1, value=0}
        ]}}})) {module=ScaleRibbonWidth001}
  ]
  rendererClasses [&NiagaraRibbonRendererProperties]
  renderers [{rendererClass=&NiagaraRibbonRendererProperties}]
 }
}
summary {
 systemName=NS_Pickaxe_Bass_PaintSplash_Era_Trail
 emitters [
  {emitterName=Galaxy_Trail, rendererClasses [&NiagaraRibbonRendererProperties]}
  {emitterName=PaintSplash001, rendererClasses [&NiagaraSpriteRendererProperties]}
  {emitterName=PaintTrail001, rendererClasses [&NiagaraRibbonRendererProperties]}
  {emitterName=Droplets, rendererClasses [&NiagaraSpriteRendererProperties]}
 ]
 userVariables [
  User.Color {type=LinearColor, description="", defaultValue {struct=LinearColor, value=<rgba 1 0 0.8389 1>}}
  User.ImpactColor {type=LinearColor, description="", defaultValue {struct=LinearColor, value=<rgba 1 1 1 1>}}
  User.IsWeakPoint {type=NiagaraBool}
  User.RibbonIDOffset {type=NiagaraInt32}
  User.SkeletalMesh {
   type=NiagaraDataInterfaceSkeletalMesh
   defaultValue {
    struct=NiagaraExt_VariableValue_DataInterface
    value {dataInterface=&self:NiagaraDataInterfaceSkeletalMesh_0}
   }
  }
  User.SpawnRate {type=NiagaraFloat, defaultValue=1, description=""}
  User.TrailColor {type=LinearColor, description="", defaultValue {struct=LinearColor, value=<rgba 1 1 1 0.25>}}
  User.TrailSecondaryColor {type=LinearColor, description="", defaultValue {struct=LinearColor, value=<rgba 10 10 10 0.25>}}
  User.UseCustomColor {type=NiagaraFloat, defaultValue=0, description=""}
  User.WaterBodyIndex {type=NiagaraFloat, defaultValue=0, description=""}
  User.WaterDepth {type=NiagaraFloat, defaultValue=0, description=""}
  User.WaterPlaneLocation {type=Vector3f, description="", defaultValue {struct=Vector3f, value=<0 0 0>}}
  User.WaterPlaneNormal {type=Vector3f, description="", defaultValue {struct=Vector3f, value=<0 0 0>}}
  User.WaveAttenuationFactor {type=NiagaraFloat, defaultValue=0, description=""}
  User.WaveTime {type=NiagaraFloat, defaultValue=0, description=""}
 ]
}
